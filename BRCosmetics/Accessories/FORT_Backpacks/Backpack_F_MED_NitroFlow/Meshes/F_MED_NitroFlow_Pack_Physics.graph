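# Physics asset for the NitroFlow backpack cosmetic, letting its dangling ropes on both sides move as physics bodies.
# Each side has a two-segment big rope chain and a single little rope, all small capsules constrained back to a capsule on the pack bone.
# Big ropes have wide swing and limited twist, while little ropes swing mostly in one plane with almost no twist.
# Left-side bones are spelled "lope" rather than "rope".
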PhysicsAsset /BRCosmetics/Accessories/FORT_Backpacks/Backpack_F_MED_NitroFlow/Meshes/F_MED_NitroFlow_Pack_Physics
bodies {
 dyn_big_lope_1_l [{shapeType=Capsule, length=1.3, radius=1, shapeName=dyn_big_lope_1_l_capsule, center=<0.9 0 0>, rotation=<rot -90 0 0>}]
 dyn_big_lope_2_l [{shapeType=Capsule, length=1.3, radius=1, shapeName=dyn_big_lope_2_l_capsule, center=<1.1 0 0>, rotation=<rot -90 0 0>}]
 dyn_big_rope_1_r [{shapeType=Capsule, length=1, radius=1, shapeName=dyn_big_rope_1_r_capsule, center=<1 0 0>, rotation=<rot -90 0 0>}]
 dyn_big_rope_2_r [{shapeType=Capsule, length=1, radius=1, shapeName=dyn_big_rope_2_r_capsule, center=<1 0 0>, rotation=<rot -90 0 0>}]
 dyn_little_lope_l [{shapeType=Capsule, length=1, radius=1, shapeName=dyn_little_rope_r_capsule, center=<3 0 0>, rotation=<rot -90 0 0>}]
 dyn_little_rope_r [{shapeType=Capsule, length=1, radius=1, shapeName=dyn_little_rope_r_capsule, center=<3 0 0>, rotation=<rot -90 0 0>}]
 pack [{shapeType=Capsule, length=1, radius=1, center=<0 -6 0>}]
}
constraints [
 {bone1Name=dyn_little_lope_l, bone2Name=pack, swing1LimitDegrees=25, swing2LimitDegrees=2, twistLimitDegrees=1}
 {bone1Name=dyn_big_lope_1_l, bone2Name=pack, swing1LimitDegrees=40, swing2LimitDegrees=45, twistLimitDegrees=10}
 {bone1Name=dyn_big_lope_2_l, bone2Name=dyn_big_lope_1_l, swing1LimitDegrees=45, swing2LimitDegrees=45, twistLimitDegrees=10}
 {bone1Name=dyn_big_rope_1_r, bone2Name=pack, swing1LimitDegrees=40, swing2LimitDegrees=45, twistLimitDegrees=10}
 {bone1Name=dyn_big_rope_2_r, bone2Name=dyn_big_rope_1_r, swing1LimitDegrees=45, swing2LimitDegrees=45, twistLimitDegrees=10}
 {bone1Name=dyn_little_rope_r, bone2Name=pack, swing1LimitDegrees=25, swing2LimitDegrees=2, twistLimitDegrees=1}
]
