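# Animation Blueprint for the NitroFlow backpack cosmetic that adds physics and airborne posing to the pack's rope bones.
# A small additive pack offset feeds a root-relative RigidBody simulation, whose gravity is updated each frame by a deprecated state function using per-state parameters.
# Post-physics bone rotations angle the main ropes when rigid body is enabled, then swing big and little ropes by a jumping-or-falling alpha, interpolated on increase.
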AnimBlueprint /BRCosmetics/Accessories/FORT_Backpacks/Backpack_F_MED_NitroFlow/Meshes/F_MED_NitroFlow_Pack_AnimBP
target_skeleton=&/F_MED_NitroFlow_Pack_Skeleton
graphs [
 AnimGraph {nodes [
   n1 = "Transform (Modify) Bone - Bone: pack"(ComponentPose=Local To Component(LocalPose=Input Pose()), Translation="0,1,0", Rotation="0,0,0") {
    translationMode=BMM_Additive
    translationSpace=BCS_BoneSpace
    alphaBoolBlend {blendOption=Linear}
    boneToModify {boneName=pack}
    translation {y=1}
   }
   n2 = RigidBody!(ComponentPose=@n1, Alpha, OverrideWorldGravity:Vector=Get GravityOverride(), ExternalForce) -> (Pose) {
    bClampLinearTranslationLimitToRefPose=true
    bOverrideWorldGravity=true
    overridePhysicsAsset=&/F_MED_NitroFlow_Pack_Physics
    simulationSpace=BaseBoneSpace
    alphaBoolBlend {blendOption=Linear}
    baseBoneRef {boneName=root}
    componentAppliedLinearAccClamp=<3000 3000 3000>
    componentLinearAccScale=<0.4 0.3 0.3>
    componentLinearVelScale=<0.4 0.3 0.3>
    overrideWorldGravity {z=-980}
    simSpaceSettings {
     maxAngularAcceleration=1000
     maxAngularVelocity=3
     maxLinearAcceleration=700
     maxLinearVelocity=500
     velocityScaleZ=0
     worldAlpha=0.5
     externalLinearDragV=<0.2 0.2>
    }
   }
   n3 = Get bRigidBodyEnabled()
   n4 = "Transform (Modify) Bone - Bone: dyn_big_rope_1_r"!(ComponentPose=@n2, bAlphaBoolEnabled:bool=@n3, Translation="0,0,0", Rotation="-0.000,-7.070,0.000", Scale="1,1,1") -> (Pose) {
    alphaInputType=Bool
    rotationMode=BMM_Additive
    rotationSpace=BCS_BoneSpace
    alphaBoolBlend {blendOption=Linear}
    boneToModify {boneName=dyn_big_rope_1_r}
    rotation=<rot -0.0001 -7.07 0>
   }
   n5 = "Transform (Modify) Bone - Bone: dyn_big_lope_1_l"!(ComponentPose=@n4, bAlphaBoolEnabled:bool=@n3, Translation="0,0,0", Rotation="0,7.07,0", Scale="1,1,1") -> (Pose) {
    alphaInputType=Bool
    rotationMode=BMM_Additive
    rotationSpace=BCS_BoneSpace
    alphaBoolBlend {blendOption=Linear}
    boneToModify {boneName=dyn_big_lope_1_l}
    rotation {yaw=7.07}
   }
   n6 = Get Is Jumping() -> (Is Jumping:double)
   n7 = "Transform (Modify) Bone - Bone: dyn_big_rope_1_r"!(ComponentPose=@n5, Alpha:float=@n6, Translation="0,0,0", Rotation="61.137,-76.742,-63.770", Scale="1,1,1") -> (Pose) {
    rotationMode=BMM_Additive
    rotationSpace=BCS_BoneSpace
    alphaBoolBlend {blendOption=Linear}
    alphaScaleBiasClamp {bInterpResult=true, interpSpeedIncreasing=5}
    boneToModify {boneName=dyn_big_rope_1_r}
    rotation=<rot 61.137 -76.7419 -63.7699>
   }
   n8 = "Transform (Modify) Bone - Bone: dyn_big_rope_2_r"(ComponentPose=@n7, Alpha=@n6, Translation="0,0,0", Rotation="30.000,0.000,0.000", Scale="1,1,1") {
    rotationMode=BMM_Additive
    rotationSpace=BCS_BoneSpace
    alphaBoolBlend {blendOption=Linear}
    alphaScaleBiasClamp {bInterpResult=true, interpSpeedIncreasing=5}
    boneToModify {boneName=dyn_big_rope_2_r}
    rotation=<rot 30 0 0>
   }
   n9 = "Transform (Modify) Bone - Bone: dyn_big_lope_1_l"!(ComponentPose=@n8, Alpha:float=@n6, Translation="0,0,0", Rotation="51.065,40.846,39.496", Scale="1,1,1") -> (Pose) {
    rotationMode=BMM_Additive
    rotationSpace=BCS_BoneSpace
    alphaBoolBlend {blendOption=Linear}
    alphaScaleBiasClamp {bInterpResult=true, interpSpeedIncreasing=5}
    boneToModify {boneName=dyn_big_lope_1_l}
    rotation=<rot 51.065 40.846 39.4959>
   }
   n10 = "Transform (Modify) Bone - Bone: dyn_big_lope_2_l"(ComponentPose=@n9, Alpha=@n6, Translation="0,0,0", Rotation="34.847,38.903,-3.492", Scale="1,1,1") {
    rotationMode=BMM_Additive
    rotationSpace=BCS_BoneSpace
    alphaBoolBlend {blendOption=Linear}
    alphaScaleBiasClamp {bInterpResult=true, interpSpeedIncreasing=5}
    boneToModify {boneName=dyn_big_lope_2_l}
    rotation=<rot 34.847 38.903 -3.492>
   }
   n11 = "Transform (Modify) Bone - Bone: dyn_little_rope_r"(ComponentPose=@n10, Alpha=@n6, Translation="0,0,0", Rotation="0.867,-29.924,9.963", Scale="1,1,1") {
    rotationMode=BMM_Additive
    rotationSpace=BCS_BoneSpace
    alphaBoolBlend {blendOption=Linear}
    alphaScaleBiasClamp {bInterpResult=true, interpSpeedIncreasing=5}
    boneToModify {boneName=dyn_little_rope_r}
    rotation=<rot 0.867 -29.924 9.963>
   }
   n12 = "Transform (Modify) Bone - Bone: dyn_little_lope_l"(ComponentPose=@n11, Alpha=@n6, Translation="0,0,0", Rotation="5.000,40.000,20.000", Scale="1,1,1") {
    rotationMode=BMM_Additive
    rotationSpace=BCS_BoneSpace
    alphaBoolBlend {blendOption=Linear}
    alphaScaleBiasClamp {bInterpResult=true, interpSpeedIncreasing=5}
    boneToModify {boneName=dyn_little_lope_l}
    rotation=<rot 5 40 20>
   }
   Output Pose(Result=Component To Local(ComponentPose=@n12)) {layerGroup=None}
  ]}
 EventGraph {nodes [
   Event BlueprintUpdateAnimation()
   > n1 = Gravity Override State Function DEPRECATED(CharacterPartAnimInstance=Self-Reference(), GameTimeInSeconds=GetGameTimeInSeconds(), GravityOverrideParams=Get Gravity Override Params())
   > Set GravityOverride(GravityOverride=@n1)
   n2 = Get bIsJumping()
   n3 = Get bIsFalling()
   > Set Is Jumping(Is Jumping:double="To Float (Boolean)"(InBool=OR Boolean(A=@n2, B=@n3))) -> (Output_Get:double)
   TryGetPawnOwner()
  ]}
]
variables [
 Gravity Override Params {
  type=GravityOverrideParamsStruct
  default {
   emoteOrMelee {gravityOverride=<0 0 -500>}
   onGround_Falling {gravityOverride=<0 0 -100>}
   onGround_Moving {gravityOverride=<0 0 -500>}
   onGround_Standing {gravityOverride=<0 0 -500>}
  }
 }
 GravityOverride {type=Vector}
 Is Jumping {type=number}
]
